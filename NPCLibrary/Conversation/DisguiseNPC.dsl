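; A conversation task node for disguised NPCs that reveals the speaking NPC's disguise when the task executes.
; It gets the participant NPC actor from the conversation context and looks for its Fortnite pawn disguise component.
; If that component is valid it calls RevealDisguise with an unknown reason, then returns the parent task's result either way.
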(fn ExecuteTaskNode
  (params (Context :type Conversation%20Context%20Structure%20%28by%20ref%29))
  (bind _returnvalue (Class|FortConversationContextLibrary|GetParticipantActorNPC Context))
  (Utilities|IsValid (Actor|GetComponentbyClass _returnvalue "/Script/FortniteGame.FortPawnComponent_Disguise")
    (:Is%20Valid
      (Class|FortPawnComponentDisguise|RevealDisguise (Actor|GetComponentbyClass _returnvalue "/Script/FortniteGame.FortPawnComponent_Disguise") 0 "Unknown"))
    (:Is%20Not%20Valid
      _))
  (bind _returnvalue_2 (|CallParentFunction Context))
  (return _returnvalue_2))
